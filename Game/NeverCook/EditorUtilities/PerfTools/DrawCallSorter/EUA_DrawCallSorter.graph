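# Editor utility actor for a draw call sorting perf tool, showing a mesh count as a severity-colored mesh with a number label.
# SetData writes a green-to-red color, the count and severity into the static mesh's custom primitive data, then updates the text.
# SetOpacity writes opacity to the same data.
# DoTextScale, also run at construction, sets text world scale from actor X scale divided by ScaleAmount.
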EditorUtilityBlueprint /Game/NeverCook/EditorUtilities/PerfTools/DrawCallSorter/EUA_DrawCallSorter
graphs [
 DoTextScale {nodes [
   DoTextScale() {is=title K2Node_FunctionEntry}
   n1 = GetActorScale3D!(self:Actor Object Reference) -> (ReturnValue:Vector)
   n2 = Get ScaleAmount() -> (ScaleAmount:double)
   n3 = "To Vector (Float)"!(InDouble:double=float / float(A=BreakVector(InVec=@n1), B=@n2)) -> (ReturnValue:Vector)
   n4 = Get TextRender() -> (TextRender:Text Render Component Object Reference)
   > SetWorldScale3D!(execute, self:Scene Component Object Reference=@n4, NewScale:Vector=@n3) -> (then)
   SetRelativeScale3D!(execute, self:Scene Component Object Reference=@n4, NewScale3D:Vector=@n3) -> (then)
   vector / vector(A:Vector=@n1, B=@n2) -> (ReturnValue:Vector)
  ]}
 EventGraph {nodes [
   Event BeginPlay()
   Event ActorBeginOverlap()
   Event Tick!() -> (OutputDelegate, then, DeltaSeconds:float)
  ]}
 SetData {nodes [
   n1 = SetData!() -> (then, NumMeshes:int, Severity:double=~) {is=title K2Node_FunctionEntry}
   > Set NumMeshes(NumMeshes:int=@n1) -> (Output_Get:int=0)
   > SetCustomPrimitiveDataFloat!(
    execute
    self:Primitive Component Object Reference=Get StaticMesh!() -> (StaticMesh:Static Mesh Component Object Reference)
    DataIndex:int=6
    Value:float=@n1.Severity
   ) -> (then)
   > SetCustomPrimitiveDataVector4!(
    execute
    self:Primitive Component Object Reference=Get StaticMesh()
    DataIndex:int=0
    Value:Vector 4 Structure=Linear Color to Vector 4!(LinearColor:Linear Color Structure="Lerp (LinearColor)"(A=(R=0.2927,G=1,B=0,A=1), B=(R=1,G=0,B=0,A=1), Alpha=@n1.Severity)) -> (Vec4:Vector 4 Structure) {is=title K2Node_MacroInstance}
   ) -> (then)
   n2 = Get NumMeshes() -> (NumMeshes:int=0)
   > SetCustomPrimitiveDataFloat(self=Get StaticMesh(), DataIndex=4, Value="To Float (Integer)"(InInt=@n2))
   > Set Text!(
    execute
    self:Text Render Component Object Reference=Get TextRender()
    Value:"Text (by ref)"="To Text (Integer)"(Value=@n2, bUseGrouping=false)
   ) -> (then) {is=title K2Node_CallFunction}
   > DoTextScale() {is=title K2Node_CallFunction}
  ]}
 SetOpacity {nodes [
   n1 = SetOpacity!() -> (then, Opacity:double=~) {is=title K2Node_FunctionEntry}
   > SetCustomPrimitiveDataFloat(self=Get StaticMesh(), DataIndex=5, Value=@n1)
  ]}
 UserConstructionScript {nodes [
   Construction Script()
   > DoTextScale()
  ]}
]
variables [
 NumMeshes {type=integer}
 ScaleAmount {type=number, default=150}
]
