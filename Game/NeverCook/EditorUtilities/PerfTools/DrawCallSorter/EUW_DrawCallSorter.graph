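# Editor utility widget that visualises actor density as transient grid cells in the level, for performance work.
# BuildMap reads actor descs, optionally within a box around the editor camera, counts non-editor-only ones per XY cell, then spawns EUA_DrawCallSorter actors coloured by min-max normalised severity.
# A self-bound DetailsView exposes settings, with opacity edits applied through a short timer.
# Buttons rebuild, hide or remove cells; closing the widget also removes them.
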EditorUtilityWidgetBlueprint /Game/NeverCook/EditorUtilities/PerfTools/DrawCallSorter/EUW_DrawCallSorter
graphs [
 BuildMap {nodes [
   BuildMap() {is=title K2Node_FunctionEntry}
   > CleanUp()
   > n1 = FetchActorDescs!(execute, self) -> (then, OutActorDescs:Array of Actor Desc Structures, ReturnValue:bool)
   > Branch(Condition=@n1.ReturnValue)
   > n2 = Get Draw Locations from Actor Descs!(execute, self, ActorDescs:Array of Actor Desc Structures=@n1) -> (then, Grids:Map of Vectors to Integers) {is=title K2Node_CallFunction}
   > n3 = Get Editor Camera Info!(execute) -> (then, CameraLocation:Vector, ForwardDirection:Vector, FOV:float, ReturnValue:bool) {is=title K2Node_CallFunction}
   > Set l_camLoc(l_camLoc:Vector=@n3) -> (Output_Get:Vector)
   > n4 = Values!(execute, TargetMap:Map of Vectors to Integers=@n2) -> (then, Values:Array of Integers)
   > n5 = Keys!(execute, TargetMap:Map of Vectors to Integers=@n2) -> (then, Keys:Array of Vectors)
   > n6 = Sequence()
   > Set l_max(l_max:int=MaxOfIntArray!(IntArray:Array of Integers=@n4) -> (IndexOfMaxValue:int=0, MaxValue:int=0).MaxValue) -> (Output_Get:int=0)
   > Set l_min(l_min:int=MinOfIntArray!(IntArray:Array of Integers=@n4) -> (IndexOfMinValue:int=0, MinValue:int=0).MinValue) -> (Output_Get:int=0)
   n7 = For Each Loop(Exec=@n6.then_1, Array:Array of Vectors=@n5) -> (Array Element:Vector)
   n8 = float - float(A=Get CellSpawnHeightOffset() -> (CellSpawnHeightOffset:double), B=5000)
   n9 = float - float!(A=BreakVector(InVec=Get l_camLoc() -> (l_camLoc:Vector)).Z, B=Get CellScaleZ() -> (CellScaleZ:double), C:double=500) -> (ReturnValue)
   n10 = Get UseCameraHeightForCellHeight() -> (UseCameraHeightForCellHeight:bool)
   n11 = vector + vector(A=@n7, B=MakeVector(Y=0, Z=SelectFloat(A=@n9, B=@n8, bPickA=@n10)))
   > n12 = SpawnTransientActor!(
    execute
    WorldContextObject:Object Reference
    ActorClass:Actor Class Reference=&/EUA_DrawCallSorter.EUA_DrawCallSorter_C
    SpawnTransform:"Transform (by ref)"="To Transform (Vector)"!(InLocation:Vector=@n11) -> (ReturnValue:Transform)
    CollisionHandlingOverride:ESpawnActorCollisionHandlingMethod Enum=AlwaysSpawn
    OwnerActor:Actor Object Reference
    InstigatorPawn:Pawn Object Reference
   ) -> (then, SpawnedActor:EUA Draw Call Sorter Object Reference)
   n13 = Get CellSize() -> (CellSize:float)
   > SetActorScale3D!(
    execute
    self:Actor Object Reference=@n12
    NewScale3D:Vector=MakeVector_NetQuantize!(X:double=@n13, Y:double=@n13, Z:double=Get CellScaleZ()) -> (ReturnValue:Vector Net Quantize Structure)
   ) -> (then)
   > n14 = Cast To EUA_DrawCallSorter!(execute, Object=@n12) -> (then, CastFailed, AsEUA Draw Call Sorter:EUA Draw Call Sorter Object Reference) {is=title K2Node_DynamicCast}
   n15 = Get l_max() -> (l_max:int=0)
   n16 = Get l_min() -> (l_min:int=0)
   n17 = "Get (a copy)"(Array:Array of Integers=@n4, Dimension 1=@n7.Array Index) -> (Output:int)
   > SetData!(
    execute
    self:EUA Draw Call Sorter Object Reference=@n14
    NumMeshes:int=@n17
    Severity:double=NormalizeToRange!(
     Value:double="To Float (Integer)"(InInt=@n17)
     RangeMin:double="To Float (Integer)"(InInt=@n16)
     RangeMax:double="To Float (Integer)"(InInt=@n15)
    ) -> (ReturnValue:double)
   ) -> (then)
   > SetOpacity!(execute, self:EUA Draw Call Sorter Object Reference=@n14, Opacity:double=Get Opacity() -> (Opacity:float)) -> (then)
   SpawnActorFromClass!(
    execute
    self:Editor Actor Subsystem Object Reference=Get EditorActorSubsystem()
    ActorClass:Actor Class Reference=&EUA_DrawCallSorter_C
    Location:Vector=@n11
    Rotation:Rotator
    bTransient:bool=true
   ) -> (then, ReturnValue:EUA Draw Call Sorter Object Reference)
   float / float(A=Get CellSize(), B=-2)
  ]}
 CleanUp {nodes [
   CleanUp() {is=title K2Node_FunctionEntry}
   > n1 = GetAllActorsOfClass!(
    execute
    WorldContextObject:Object Reference
    ActorClass:Actor Class Reference=&EUA_DrawCallSorter_C
   ) -> (then, OutActors:Array of EUA Draw Call Sorter Object References)
   > DestroyActors!(execute, self:Editor Actor Subsystem Object Reference=Get EditorActorSubsystem(), ActorsToDestroy:Array of Actor Object References=@n1) -> (then, ReturnValue:bool)
  ]}
 EventGraph {nodes [
   "On Clicked (TA_Generic01_Button)"!() -> (OutputDelegate, then, widget:User Widget Object Reference, event:Pointer Event Structure) {is=title K2Node_ComponentBoundEvent}
   > BuildMap() {is=title K2Node_CallFunction}
   "On Clicked (btnClean)"!() -> (OutputDelegate, then, widget:User Widget Object Reference, event:Pointer Event Structure) {is=title K2Node_ComponentBoundEvent}
   > CleanUp() {is=title K2Node_CallFunction}
   "On Clicked (btnToggleVis)"!() -> (OutputDelegate, then, widget:User Widget Object Reference, event:Pointer Event Structure) {is=title K2Node_ComponentBoundEvent}
   > n1 = GetAllActorsOfClass(ActorClass=&EUA_DrawCallSorter_C)
   > n2 = For Each Loop(Array:Array of EUA Draw Call Sorter Object References=@n1) -> (Array Element:EUA Draw Call Sorter Object Reference)
   > SetIsTemporarilyHiddenInEditor!(
    execute
    self:Actor Object Reference=@n2
    bIsHidden:bool=NOT Boolean(A=IsTemporarilyHiddenInEditor!(self:Actor Object Reference=@n2, bIncludeParent:bool=true) -> (ReturnValue:bool))
   ) -> (then)
   n3 = "On Property Changed (DetailsView_0)"!() -> (OutputDelegate, then, PropertyName:Name=None) {is=title K2Node_ComponentBoundEvent}
   > Branch(Condition="Equal (Name)"(A=@n3.PropertyName, B=Opacity))
   > n4 = Branch(Condition=Is Valid Timer Handle!(Handle:Timer Handle Structure=Get OpacityTimerHandle() -> (OpacityTimerHandle:Timer Handle Structure)) -> (ReturnValue:bool))
   n5 = Set Timer by Event(execute=@n4.else, Delegate=Create Event(), Time=0.05, bLooping=false, bMaxOncePerFrame=true)
   > Set OpacityTimerHandle(OpacityTimerHandle:Timer Handle Structure=@n5) -> (Output_Get:Timer Handle Structure)
   Event Construct()
   > SetObject(self=Get DetailsView_0() -> (DetailsView_0:Details View Object Reference), NewObject=Self-Reference())
   Event Destruct()
   > CleanUp()
  ]}
 FetchActorDescs {nodes [
   FetchActorDescs() {is=title K2Node_FunctionEntry}
   > n1 = Branch(Condition=Get RelativeToCamera() -> (RelativeToCamera:bool))
   > n2 = Get Editor Camera Info()
   n3 = Get RadiusAroundCamera() -> (RadiusAroundCamera:double)
   > n4 = GetIntersectingActorDescs!(
    execute
    InBox:"Box Structure (by ref)"=MakeBoxFromCenterSize!(Center:Vector=@n2, Dimensions:Vector=MakeVector(X=@n3, Y=@n3, Z=999999)) -> (ReturnValue:Box Structure)
   ) -> (then, OutActorDescs:Array of Actor Desc Structures, ReturnValue:bool)
   > Return Node!(execute, OutActorDescs:Array of Actor Desc Structures=@n4, ReturnValue:bool=@n4.ReturnValue)
   n5 = GetActorDescs!(execute=@n1.else) -> (then, OutActorDescs:Array of Actor Desc Structures, ReturnValue:bool)
   > Return Node(OutActorDescs=@n5, ReturnValue=@n5.ReturnValue)
  ]}
 Get Draw Locations from Actor Descs {nodes [
   n1 = Get Draw Locations from Actor Descs!() -> (then, ActorDescs:Array of Actor Desc Structures) {is=title K2Node_FunctionEntry}
   > n2 = For Each Loop(Array:Array of Actor Desc Structures=@n1) -> (Array Element:Actor Desc Structure)
   Return Node!(execute=@n2.Completed, Grids:Map of Vectors to Integers=Get l_grids() -> (l_grids:Map of Vectors to Integers))
   n3 = Break Actor Desc!(ActorDesc:"Actor Desc Structure (by ref)"=@n2) -> (
    Guid:Guid Structure
    NativeClass:Object Class Reference
    Class:Soft Object Path Structure
    Name:Name
    Label:Name
    Bounds:Box Structure
    RuntimeGrid:Name
    bIsSpatiallyLoaded:bool=~
    bActorIsEditorOnly:bool=~
    ActorPackage:Name
    ActorPath:Name
    FolderPath:Name
    DataLayerAssets:Array of Soft Object Path Structures
   ) {is=title K2Node_BreakStruct}
   > n4 = Branch(Condition=@n3.bActorIsEditorOnly)
   n5 = "Get Center (Box)"!(InBox:"Box Structure (by ref)"=@n3.Bounds) -> (ReturnValue:Vector) {is=title K2Node_CallFunction}
   Set l_loc(
    execute=@n4.else
    l_loc:Vector=Vector_SnappedToGrid!(InVect:Vector=vector * vector(A=@n5, B:Vector="1,1,0") -> (ReturnValue=~), InGridSize:float=Get CellSize()) -> (ReturnValue:Vector)
   ) -> (Output_Get:Vector)
   n6 = Find(TargetMap:Map of Vectors to Integers=Get l_grids(), Key:"Vector (by ref)"=Get l_loc() -> (l_loc:Vector)) -> (Value:int)
   > n7 = Branch(Condition=@n6.ReturnValue)
   > Add!(execute, TargetMap:Map of Vectors to Integers=Get l_grids(), Key:"Vector (by ref)"=Get l_loc(), Value:"Integer (by ref)"=int + int(A=@n6)) -> (then) {is=title K2Node_CallFunction}
   Add(execute=@n7.else, TargetMap=Get l_grids(), Key=Get l_loc(), Value=1)
  ]}
 UpdateOpacity {nodes [
   UpdateOpacity() {is=title K2Node_FunctionEntry}
   > Clear and Invalidate Timer by Handle!(execute, WorldContextObject:Object Reference, Handle=Get OpacityTimerHandle()) -> (then)
   > n1 = GetAllActorsOfClass(ActorClass=&EUA_DrawCallSorter_C)
   > n2 = For Each Loop(Array:Array of EUA Draw Call Sorter Object References=@n1) -> (Array Element:EUA Draw Call Sorter Object Reference)
   > SetOpacity(self=@n2, Opacity=Get Opacity())
  ]}
]
variables [
 CellScaleZ {type=number, default=15000}
 CellSize {type=number, default=1000}
 CellSpawnHeightOffset {type=number}
 Opacity {type=number}
 OpacityTimerHandle {type=TimerHandle}
 RadiusAroundCamera {type=number, default=55000}
 RelativeToCamera {type=boolean, default=true}
 UseCameraHeightForCellHeight {type=boolean, default=true}
]
widgets {
 info {rootWidgetClass=&StackBox, widgetCount=5}
 widgets [
  StackBox_0 {widgetClassPath=&StackBox, widget=StackBox_0, properties {orientation=Orient_Vertical}}
  TA_Generic01_Button {
   widgetClassPath=&TA_Generic01_Button
   parent=StackBox_0
   widget=TA_Generic01_Button
   properties {
    label=Display actor counts in Grid
    toolTipText="Spawn Cells based on settings below. Each cell shows the number of actors inside it.  Can be slow."
   }
   slot {class=StackBoxSlot}
  }
  btnToggleVis {
   widgetClassPath=&TA_Generic01_Button
   parent=StackBox_0
   widget=btnToggleVis
   properties {label=Toggle Visibility of Cells, toolTipText=Temp toggle the visibility of cells easily., buttonColor_Default=<rgba 0.6354 0.4576 0.0495 1>}
   slot {class=StackBoxSlot}
  }
  DetailsView_0 {
   widgetClassPath=&DetailsView
   parent=StackBox_0
   widget=DetailsView_0
   properties {categoriesToShow [nocategory]}
   slot {class=StackBoxSlot, properties {size {sizeRule=Fill, value=1}}}
  }
  btnClean {
   widgetClassPath=&TA_Generic01_Button
   parent=StackBox_0
   widget=btnClean
   properties {label=Clean Up, toolTipText=Remove all cells., buttonColor_Default=<rgba 1 0.0778 0 1>}
   slot {class=StackBoxSlot}
  }
 ]
}
