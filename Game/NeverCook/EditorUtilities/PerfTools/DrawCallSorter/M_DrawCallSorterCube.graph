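# Unlit translucent editor material for the DrawCallSorter perf tool's cubes, drawn without depth test to visualize a per-cube severity.
# Desaturated scene color is tinted by a green-orange-red gradient, with power-adjusted Severity apparently feeding it through reroutes.
# An opacity calculation separates the upward top face from the edges and scales by severity.
# Color, NumMeshes, Opacity Mult and Severity use custom primitive data; additional nodes render NumMeshes as debug digits and test bounds behind translucency.
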Material /Game/NeverCook/EditorUtilities/PerfTools/DrawCallSorter/M_DrawCallSorterCube
graph {
 properties {
  bAutomaticallySetUsageInEditor=false
  bDisableDepthTest=true
  blendMode=BLEND_Translucent
  pixelDepthOffsetMode=PDOM_Legacy
  shadingModel=MSM_Unlit
  usedShadingModels=Unlit
 }
 nodes [
  Constant() {r=1}
  n1 = ObjectPositionWS()
  n2 = ObjectBounds()
  n3 = MaterialFunctionCall(Float3=Subtract(A=@n1, B=@n2)) {materialFunction=&BreakOutFloat3Components}
  n4 = MaterialFunctionCall(Float3=Add(A=@n2, B=@n1)) {materialFunction=&BreakOutFloat3Components}
  n5 = MaterialFunctionCall(Float3=MaterialFunctionCall() {materialFunction=&WorldPositionBehindTranslucency}.Result) {materialFunction=&BreakOutFloat3Components}
  n6 = InverseLinearInterpolate(A=@n3.R, B=@n4.R, Value=@n5.R)
  n7 = InverseLinearInterpolate(A=@n3.G, B=@n4.G, Value=@n5.G)
  n8 = Constant()
  If(A=Add(A=Ceil(None=@n6), B=Ceil(None=@n7)), "A > B"=@n8, "A == B"=Constant() {r=1}, "A < B"=@n8) {constB=2}
  LinearInterpolate(
   A=VectorParameter() {bUseCustomPrimitiveData=true, parameterName=Color, defaultValue=<rgba 0.7448 0.1901 0.1901 1>}.RGB
   Alpha=MaterialFunctionCall(
    Number=ScalarParameter() {bUseCustomPrimitiveData=true, parameterName=NumMeshes, primitiveDataIndex=4}
    MaximumNumberOfDigits=ScalarParameter() {defaultValue=4, parameterName=MaxDig}
   ) {materialFunction=&DebugScalarValues}.Result
  )
  LinearInterpolate()
  n9 = Desaturation(None=SceneColor!(UVs).RGB)
  n10 = ScalarParameter() {defaultValue=0.25, parameterName=SceneMin}
  Max(A=@n9, B=@n10)
  n11 = DotProduct(A=PixelNormalWS(), B=Constant3Vector() {constant=<rgba 0 0 1 0>})
  n12 = If(A=WorldPosition().Z, B=ComponentMask(None=ObjectPositionWS()) {b=true}, "A > B"=Constant() {r=1}, "A < B"=Constant())
  n13 = Constant()
  n14 = NamedRerouteUsage()
  Multiply(
   A=LinearInterpolate(
    A=ScalarParameter() {defaultValue=0.1, parameterName=Opacity Mult Edge}
    B=ScalarParameter() {bUseCustomPrimitiveData=true, defaultValue=1, parameterName=Opacity Mult, primitiveDataIndex=5}
    Alpha=If(A=@n11, "A > B"=@n12, "A < B"=@n13) {constB=0.9}
   )
   B=Saturate(None=LinearInterpolate(Alpha=@n14) {constA=0.25, constB=2})
  )
  NamedRerouteDeclaration(None=Power(
    Base=ScalarParameter() {bUseCustomPrimitiveData=true, defaultValue=0.1, parameterName=Severity, primitiveDataIndex=6}
    Exp=ScalarParameter() {defaultValue=0.4, parameterName=SevPower}
   )) {name=Router_Severity, nodeColor=<rgba 0 1 0.3438 1>}
  n15 = VectorParameter() {parameterName=Color0, defaultValue=<rgba 0.0712 0.7448 0 1>}
  n16 = VectorParameter() {parameterName=Color1, defaultValue=<rgba 0.7448 0.2758 0 1>}
  n17 = VectorParameter() {parameterName=Color2, defaultValue=<rgba 1 0 0 1>}
  n18 = NamedRerouteUsage()
  n19 = Constant() {r=1}
  ShadingPathSwitch(
   Default=Multiply(
    A=LinearInterpolate(A=@n10, Alpha=@n9)
    B=MaterialFunctionCall(1=@n15.RGB, 2=@n16.RGB, 3=@n17.RGB, A=@n18) {materialFunction=&Lerp_Multiple_Float3}.Lerp 3 Inputs
   )
   Forward=@n19
   Mobile=@n19
  )
 ]
}
